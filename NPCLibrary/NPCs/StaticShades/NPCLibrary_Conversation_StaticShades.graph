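# Conversation database for the StaticShades NPC (Jittershock), supplying the NPC's dialogue when a player talks to them.
# The entry point feeds one Speech node with a default line and contextual messages chosen by requirement and priority.
# Variants cover a first-ever meeting, a reaction to the SunShine outfit, and general lines gated on match conversation count.
# The speech then branches through link nodes to the shared NPC Library Quests and BasicServices conversations.
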ConversationDatabase /NPCLibrary/NPCs/StaticShades/NPCLibrary_Conversation_StaticShades
entry_points [{tag=Conversation.Entry.StaticShades, destinations [#1]}]
nodes [
 ConversationEntryPointNode_7 {
  id=#1
  class=ConversationEntryPointNode
  outputs [#2]
  properties {entryTag {tagName=Conversation.Entry.StaticShades}}
 }
 Speech_C_4 {
  id=#2
  class=Speech_C
  outputs [#3, #4]
  properties {
   nodeName=""
   generalConfig {
    defaultMessage="Oh wow... THIS PLACE IS SO COOL!"
    contextualMessages [
     {
      message="I'm not jittery, bud. I'm Jittershock!"
      contextRequirements [{requirement {class=FortControllerRequirement_IsFirstEverConversationWithNPC}}]
      priority {curve {rowName=Tandem.Speech.Priority.FirstEverInteraction}}
     }
     {
      message="Buddy, I think this place is gonna make our dreams come true."
      contextRequirements [{requirement {class=FortControllerRequirement_HasCID, properties {allowedCIds [/BRCosmetics/Athena/Items/Cosmetics/Characters/Character_SunShine]}}}]
      priority {curve {rowName=Tandem.Speech.Priority.OutfitReaction}}
     }
     {
      message="Oh wow... THIS PLACE IS SO COOL!"
      contextRequirements [{requirement {class=FortControllerRequirement_CheckNPCMatchConversationCount, properties {nPCUniqueIdTag {tagName=AISpawnerData.Type.Tandem.StaticShades}}}}]
      priority {curve {rowName=None}}
     }
     {
      message=I wonder what neat metals they got down in that Underworld.
      contextRequirements [{requirement {class=FortControllerRequirement_CheckNPCMatchConversationCount, properties {nPCUniqueIdTag {tagName=AISpawnerData.Type.Tandem.StaticShades}}}}]
      priority {curve {rowName=None}}
     }
     {
      message="What's your favorite snack? Nickel's really sweet, but gold... mmm..."
      contextRequirements [{requirement {class=FortControllerRequirement_CheckNPCMatchConversationCount, properties {nPCUniqueIdTag {tagName=AISpawnerData.Type.Tandem.StaticShades}}}}]
      priority {curve {rowName=None}}
     }
     {
      message=I bet that Pandora's Box tastes realllly good. Or maybe it would just blow me up.
      contextRequirements [{requirement {class=FortControllerRequirement_CheckNPCMatchConversationCount, properties {nPCUniqueIdTag {tagName=AISpawnerData.Type.Tandem.StaticShades}}}}]
      priority {curve {rowName=None}}
     }
    ]
   }
  }
 }
 ConversationLinkNode_0 {id=#4, class=ConversationLinkNode, properties {remoteEntryTag {tagName=Conversation.Entry.BasicServices}}}
 ConversationLinkNode_1 {id=#3, class=ConversationLinkNode, properties {remoteEntryTag {tagName=Conversation.Entry.NPCLibrary.Quests}}}
]
